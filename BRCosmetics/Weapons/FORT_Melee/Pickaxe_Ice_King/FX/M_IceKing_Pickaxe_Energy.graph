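# Translucent unlit material for energy effects on the Ice King pickaxe, drawing glowing, scrolling smoke.
# A tiled smoke texture pans along V on the second UV channel; its green channel plus vertex red indexes a crystal gradient for color.
# Smoke-driven intensity pushes that color into HDR emissive through DeriveHDRfromLDR.
# Opacity combines smoke and vertex red, then multiplies by first-channel UV falloffs that fade the edges.
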Material /BRCosmetics/Weapons/FORT_Melee/Pickaxe_Ice_King/FX/M_IceKing_Pickaxe_Energy
graph {
 properties {blendMode=BLEND_Translucent, pixelDepthOffsetMode=PDOM_Legacy, shadingModel=MSM_Unlit, usedShadingModels=Unlit}
 nodes [
  n1 = TextureCoordinate() {coordinateIndex=1}
  n2 = AppendVector(A=ScalarParameter() {defaultValue=2, parameterName=UScale}, B=ScalarParameter() {defaultValue=0.125, parameterName=VScale})
  n3 = Constant()
  n4 = Multiply(A=Time(), B=ScalarParameter() {defaultValue=-0.1726, parameterName=Speed})
  n5 = TextureSample(UVs=Add(A=Multiply(A=@n1, B=@n2), B=AppendVector(A=@n3, B=@n4))) {texture=&T_Inky_Smoke_Tile}
  n6 = VertexColor()
  n7 = TextureSampleParameter2D(UVs=Add(A=@n5.G, B=@n6.R)) {parameterName=GradientTexture, texture=/Game/Global/Textures/Default/Masks_and_Gradients/HeistColorGradients/CrystalGradient}
  n8 = ScalarParameter() {defaultValue=1, parameterName=HDR_Intensity_Min}
  n9 = ScalarParameter() {defaultValue=10, parameterName=HDR_Intensity_Max}
  n10 = Power(Base=@n5.G, Exp=ScalarParameter() {defaultValue=2, parameterName=HDR Contrast})
  MaterialFunctionCall(Emissive=MaterialFunctionCall(LDR Input=Clamp(None=@n7.RGB), HDR Tint=@n7.RGB, HDR Intensity Multiplier=LinearInterpolate(A=@n8, B=@n9, Alpha=@n10) {constB=10}) {materialFunction=&DeriveHDRfromLDR}.RemappedRange) {materialFunction=&MF_FX_EmissiveAdjustments}
  n11 = Add(
   A=Power(Base=@n5.G, Exp=ScalarParameter() {defaultValue=0.5, parameterName=Opacity Contrast}) {constExponent=0.5}
   B=Multiply(A=@n6.R, B=ScalarParameter() {defaultValue=1, parameterName=Vertex Intensity})
  )
  n12 = TextureCoordinate()
  n13 = Power(Base=OneMinus(None=ComponentMask(None=@n12) {g=true}), Exp=ScalarParameter() {defaultValue=5, parameterName=ContrastExp})
  n14 = ComponentMask(None=TextureCoordinate()) {r=true}
  n15 = OneMinus(None=@n14)
  n16 = ScalarParameter() {defaultValue=3, parameterName=Bottom Edge Contrast}
  n17 = Multiply(A=Multiply(A=@n14, B=Power(Base=@n15, Exp=@n16))) {constB=10}
  Multiply(A=ScalarParameter() {defaultValue=1, parameterName=Overall Opacity}, B=Multiply(A=Multiply(A=@n11, B=@n13), B=Clamp(None=@n17)))
 ]
}
