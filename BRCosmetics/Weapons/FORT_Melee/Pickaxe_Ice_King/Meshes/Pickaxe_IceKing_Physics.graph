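# Physics asset for the Ice King pickaxe mesh, setting up simulated motion for its hanging straps and tassels.
# A cloth_base body built from three capsules anchors two five-link strap chains (A and B) and two two-link tassel chains, all made of box bodies.
# Strap roots swing freely on the base, while links further along use limited swing and twist.
# Tassels have tighter swing limits and very small twist limits.
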PhysicsAsset /BRCosmetics/Weapons/FORT_Melee/Pickaxe_Ice_King/Meshes/Pickaxe_IceKing_Physics
bodies {
 cloth_base [
  {shapeType=Capsule, length=36.755, radius=3.505, center=<0 0 -5.4064>, rotation=<rot 0 -90 0>}
  {shapeType=Capsule, length=7, radius=3.75, center=<0.044 -0.805 -16.3793>, rotation=<rot -0.0001 -1.9356 89.9995>}
  {shapeType=Capsule, length=4.75, radius=3.75, center=<0.5382 -0.9626 -16.3793>, rotation=<rot 0.0005 90.0314 89.9996>}
 ]
 dyn_strapA_01 [{shapeType=Box, extentX=1.01, extentY=5.01, extentZ=4.01, center=<6.0936 0.0004 0>, rotation=<rot 0 -89.9999 0>}]
 dyn_strapA_02 [{shapeType=Box, extentX=1.01, extentY=3.3989, extentZ=8.5767, center=<3.3335 -0.1529 -0.0216>, rotation=<rot 0 -90 -89.4195>}]
 dyn_strapA_03 [{shapeType=Box, extentX=1.01, extentY=3.3971, extentZ=8.5749, center=<3.0407 -0.1527 -0.1312>, rotation=<rot 0 -90 -89.2629>}]
 dyn_strapA_04 [{shapeType=Box, extentX=1.3413, extentY=4.4221, extentZ=9.8666, center=<3.6512 -0.1582 -0.2892>, rotation=<rot 0 -90.0651 -88.9281>}]
 dyn_strapA_05 [{shapeType=Box, extentX=1.01, extentY=10.01, extentZ=5.01, center=<4.4389 -0.0004 0>, rotation=<rot 0 -90 0>}]
 dyn_strapB_01 [{shapeType=Box, extentX=1.01, extentY=3.9278, extentZ=3.8935, center=<5.1674 0.0278 0.125>, rotation=<rot 0 -89.9999 0>}]
 dyn_strapB_02 [{shapeType=Box, extentX=1.01, extentY=3.5983, extentZ=6.9547, center=<2.5073 0.0276 0.0167>, rotation=<rot 0 89.9999 -90.543>}]
 dyn_strapB_03 [{shapeType=Box, extentX=1.01, extentY=8.5696, extentZ=3.3935, center=<3.4724 0.0277 0.125>, rotation=<rot 0 -90 0>}]
 dyn_strapB_04 [{shapeType=Box, extentX=1.3413, extentY=4.9221, extentZ=9.3649, center=<3.3657 0.0221 0.1162>, rotation=<rot 0 -90.0651 -89.9456>}]
 dyn_strapB_05 [{shapeType=Box, extentX=1.01, extentY=10.01, extentZ=5.01, center=<3.9813 0.0001 0>, rotation=<rot 0 -90 0>}]
 dyn_tassel_bot_01 [{shapeType=Box, extentX=2.4021, extentY=2.8213, extentZ=0.5, center=<0.5231 0.0001 0>}]
 dyn_tassel_bot_02 [{shapeType=Box, extentX=3.0655, extentY=0.5, extentZ=2.5538, center=<0.8804 0 0>, rotation=<rot 0 0 90>}]
 dyn_tassel_top_01 [{shapeType=Box, extentX=2.276, extentY=0.5, extentZ=2.5051, center=<0.5228 0 0>, rotation=<rot 0 180 -89.9996>}]
 dyn_tassel_top_02 [{shapeType=Box, extentX=2.8384, extentY=2.713, extentZ=0.5, center=<0.6417 -0.0001 0>}]
}
constraints [
 {bone1Name=dyn_tassel_top_01, bone2Name=cloth_base, swing1LimitDegrees=28.5714, swing2LimitDegrees=67.5714, twistLimitDegrees=5}
 {bone1Name=dyn_tassel_top_02, bone2Name=dyn_tassel_top_01, swing1LimitDegrees=20, swing2LimitDegrees=35, twistLimitDegrees=5}
 {bone1Name=dyn_tassel_bot_01, bone2Name=cloth_base, swing1LimitDegrees=23.4286, swing2LimitDegrees=45.5714, twistLimitDegrees=5}
 {bone1Name=dyn_tassel_bot_02, bone2Name=dyn_tassel_bot_01, swing1LimitDegrees=20, swing2LimitDegrees=40.4286, twistLimitDegrees=5}
 {
  bone1Name=dyn_strapA_01
  bone2Name=cloth_base
  swing1LimitDegrees=80
  swing1Motion=Free
  swing2LimitDegrees=80
  swing2Motion=Free
  twistLimitDegrees=90
 }
 {bone1Name=dyn_strapA_02, bone2Name=dyn_strapA_01, swing1LimitDegrees=150, swing2LimitDegrees=150, twistLimitDegrees=60}
 {bone1Name=dyn_strapA_03, bone2Name=dyn_strapA_02, swing1LimitDegrees=55, swing2LimitDegrees=55, twistLimitDegrees=20}
 {bone1Name=dyn_strapA_04, bone2Name=dyn_strapA_03, swing1LimitDegrees=65, swing2LimitDegrees=65, twistLimitDegrees=25}
 {bone1Name=dyn_strapA_05, bone2Name=dyn_strapA_04, swing1LimitDegrees=70, swing2LimitDegrees=70, twistLimitDegrees=25}
 {
  bone1Name=dyn_strapB_01
  bone2Name=cloth_base
  swing1LimitDegrees=80
  swing1Motion=Free
  swing2LimitDegrees=80
  swing2Motion=Free
  twistLimitDegrees=90
 }
 {bone1Name=dyn_strapB_02, bone2Name=dyn_strapB_01, swing1LimitDegrees=150, swing2LimitDegrees=150, twistLimitDegrees=60}
 {bone1Name=dyn_strapB_03, bone2Name=dyn_strapB_02, swing1LimitDegrees=55, swing2LimitDegrees=55, twistLimitDegrees=20}
 {bone1Name=dyn_strapB_04, bone2Name=dyn_strapB_03, swing1LimitDegrees=65, swing2LimitDegrees=65, twistLimitDegrees=25}
 {bone1Name=dyn_strapB_05, bone2Name=dyn_strapB_04, swing1LimitDegrees=70, swing2LimitDegrees=70, twistLimitDegrees=25}
]
