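# Animation Blueprint for the Ice King pickaxe that adds secondary motion to its dangling straps.
# The cached input pose feeds a RigidBody simulation in handle bone space with overridden gravity, used when bRigidBodyEnabled is true.
# Otherwise, bCanPlayCustomAnimations picks one of two additive bone-offset poses for the strap roots, which then runs through Trail controllers on both five-bone chains.
# The event graph has no logic, and the pose Sequence Player is unconnected.
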AnimBlueprint /BRCosmetics/Weapons/FORT_Melee/Pickaxe_Ice_King/Meshes/Pickaxe_IceKing_AnimBP
target_skeleton=&/Pickaxe_IceKing_Skeleton
graphs [
 AnimGraph {nodes [
   n1 = "Transform (Modify) Bone - Bone: dyn_strapA_01"(
    ComponentPose=Local To Component(LocalPose=Use cached pose 'AnimCache'())
    Alpha=1
    Translation="-4,2,0"
    Rotation="-13,-25,0"
    Scale="1,1,1"
   ) {
    rotationMode=BMM_Additive
    rotationSpace=BCS_BoneSpace
    translationMode=BMM_Additive
    translationSpace=BCS_BoneSpace
    alphaBoolBlend {blendOption=Linear}
    boneToModify {boneName=dyn_strapA_01}
    rotation {pitch=-13, yaw=-25}
    translation=<-4 2>
   }
   n2 = "Transform (Modify) Bone - Bone: dyn_strapA_02"(ComponentPose=@n1, Alpha=1, Translation="0,0,0", Rotation="-13,-30,0.000", Scale="1,1,1") {
    rotationMode=BMM_Additive
    rotationSpace=BCS_BoneSpace
    translationMode=BMM_Additive
    translationSpace=BCS_BoneSpace
    alphaBoolBlend {blendOption=Linear}
    boneToModify {boneName=dyn_strapA_02}
    rotation {pitch=-13, yaw=-30}
   }
   n3 = "Transform (Modify) Bone - Bone: dyn_strapB_01"(ComponentPose=@n2, Alpha=1, Translation="-4,2,0", Rotation="-15,-25,0", Scale="1,1,1") {
    rotationMode=BMM_Additive
    rotationSpace=BCS_BoneSpace
    translationMode=BMM_Additive
    translationSpace=BCS_BoneSpace
    alphaBoolBlend {blendOption=Linear}
    boneToModify {boneName=dyn_strapB_01}
    rotation {pitch=-15, yaw=-25}
    translation=<-4 2>
   }
   n4 = "Transform (Modify) Bone - Bone: dyn_strapB_02"(ComponentPose=@n3, Alpha=1, Translation="0,0,0", Rotation="-15,-30,0.000", Scale="1,1,1") {
    rotationMode=BMM_Additive
    rotationSpace=BCS_BoneSpace
    translationMode=BMM_Additive
    translationSpace=BCS_BoneSpace
    alphaBoolBlend {blendOption=Linear}
    boneToModify {boneName=dyn_strapB_02}
    rotation {pitch=-15, yaw=-30}
   }
   n5 = Component To Local(ComponentPose=@n4)
   n6 = "Transform (Modify) Bone - Bone: dyn_strapA_01"(
    ComponentPose=Local To Component(LocalPose=Use cached pose 'AnimCache'())
    Alpha=1
    Translation="-5.2,0,-2.2"
    Rotation="-30,10,0"
    Scale="1,1,1"
   ) {
    rotationMode=BMM_Additive
    rotationSpace=BCS_BoneSpace
    translationMode=BMM_Additive
    translationSpace=BCS_BoneSpace
    alphaBoolBlend {blendOption=Linear}
    boneToModify {boneName=dyn_strapA_01}
    rotation {pitch=-30, yaw=10}
    translation {x=-5.2, z=-2.2}
   }
   n7 = "Transform (Modify) Bone - Bone: dyn_strapA_02"(ComponentPose=@n6, Alpha=1, Translation="0.6,0,5", Rotation="-20.000,0,0.000", Scale="1,1,1") {
    rotationMode=BMM_Additive
    rotationSpace=BCS_BoneSpace
    translationMode=BMM_Additive
    translationSpace=BCS_BoneSpace
    alphaBoolBlend {blendOption=Linear}
    boneToModify {boneName=dyn_strapA_02}
    rotation {pitch=-20}
    translation {x=0.6, z=5}
   }
   n8 = "Transform (Modify) Bone - Bone: dyn_strapB_01"(ComponentPose=@n7, Alpha=1, Translation="0,0,0", Rotation="-20,10,0", Scale="1,1,1") {
    rotationMode=BMM_Additive
    rotationSpace=BCS_BoneSpace
    translationMode=BMM_Additive
    translationSpace=BCS_BoneSpace
    alphaBoolBlend {blendOption=Linear}
    boneToModify {boneName=dyn_strapB_01}
    rotation {pitch=-20, yaw=10}
   }
   n9 = "Transform (Modify) Bone - Bone: dyn_strapB_02"(ComponentPose=@n8, Alpha=1, Translation="0,0,0", Rotation="-25,0,0.000", Scale="1,1,1") {
    rotationMode=BMM_Additive
    rotationSpace=BCS_BoneSpace
    translationMode=BMM_Additive
    translationSpace=BCS_BoneSpace
    alphaBoolBlend {blendOption=Linear}
    boneToModify {boneName=dyn_strapB_02}
    rotation {pitch=-25}
   }
   n10 = Component To Local(ComponentPose=@n9)
   n11 = Get bCanPlayCustomAnimations() -> (bCanPlayCustomAnimations:bool)
   n12 = Trail controller - Bone: dyn_strapA_05!(ComponentPose=Local To Component(LocalPose=Blend Poses by bool(BlendPose_0=@n10, BlendPose_1=@n5, bActiveValue=@n11) {blendType=Linear})) -> (Pose) {
    chainLength=5
    relaxationSpeedScale=1.5
    rotationLimits [{limitMax=<180 180 180>}, {limitMax=<180 180 180>}, {limitMax=<180 180 180>}, {limitMax=<180 180 180>}, {limitMax=<180 180 180>}]
    rotationOffsets [<0 0 0>, <0 0 0>, <0 0 0>, <0 0 0>, <0 0 0>]
    trailBone {boneName=dyn_strapA_05}
    trailRelaxationSpeed {editorCurveData {keys [<at 0 8 RCIM_Linear>,<at 1 4 RCIM_Linear>]}}
   }
   n13 = Trail controller - Bone: dyn_strapB_05!(ComponentPose=@n12) -> (Pose) {
    chainLength=5
    rotationLimits [{limitMax=<180 180 180>}, {limitMax=<180 180 180>}, {limitMax=<180 180 180>}, {limitMax=<180 180 180>}, {limitMax=<180 180 180>}]
    rotationOffsets [<0 0 0>, <0 0 0>, <0 0 0>, <0 0 0>, <0 0 0>]
    trailBone {boneName=dyn_strapB_05}
    trailRelaxationSpeed {editorCurveData {keys [<at 0 8 RCIM_Linear>,<at 1 4 RCIM_Linear>]}}
   }
   n14 = RigidBody(ComponentPose=Local To Component(LocalPose=Use cached pose 'AnimCache'())) {
    bClampLinearTranslationLimitToRefPose=true
    bOverrideWorldGravity=true
    overridePhysicsAsset=&/Pickaxe_IceKing_Physics
    simulationSpace=BaseBoneSpace
    alphaBoolBlend {blendOption=Linear}
    baseBoneRef {boneName=Melee_TwoHanded_Handle}
    componentAppliedLinearAccClamp=<500 500 500>
    componentLinearAccScale=<0.7 0.7 0.7>
    componentLinearVelScale=<1 1 1>
    overrideWorldGravity {z=-1000}
   }
   Output Pose(Result=Blend Poses by bool(BlendPose_0=Component To Local(ComponentPose=@n14), BlendPose_1=Component To Local(ComponentPose=@n13), bActiveValue=Get bRigidBodyEnabled()) {blendType=Linear}) {layerGroup=None}
   AnimGraphNode_SubInput_0 = Input Pose()
   Save cached pose 'AnimCache'(Pose=@AnimGraphNode_SubInput_0) {cacheName=AnimCache}
   Sequence Player 'Pickaxe_IceKing_Pose'() {sequence=&/Pickaxe_IceKing_Pose}
  ]}
 EventGraph {nodes [Event BlueprintUpdateAnimation()]}
]
variables []
